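# Animation Blueprint for the NobleSpear female medium body that follows the parent mesh's pose and takes facial bones from the character's head mesh.
# On customization load it caches the pawn's Head part mesh, whose pose is layered over the copied parent pose from the listed chin, jaw and facial-root bones.
# The blend feeds the NobleSpear Control Rig, then an additive bone-space offset on attach_backpack.
# Two fix_neck_02 offset nodes are not connected to the output.
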AnimBlueprint /BRCosmetics/Characters/Player/Female/Medium/Bodies/F_MED_NobleSpear/Meshes/F_MED_NobleSpear_AnimBP
target_skeleton=&/F_MED_NobleSpear_Skeleton
graphs [
 AnimGraph {nodes [
   Copy Pose From Mesh() {bUseAttachedParent=true}
   n1 = "Transform (Modify) Bone - Bone: fix_neck_02"!(ComponentPose, bAlphaBoolEnabled:bool=True, Translation="0,0.5,0", Rotation="0,0,0", Scale="1,1,1") -> (Pose) {
    alphaInputType=Bool
    rotationMode=BMM_Additive
    rotationSpace=BCS_BoneSpace
    translationMode=BMM_Additive
    translationSpace=BCS_BoneSpace
    alphaBoolBlend {blendOption=Linear}
    boneToModify {boneName=fix_neck_02}
    translation {y=0.5}
   }
   "Transform (Modify) Bone - Bone: fix_neck_02"(ComponentPose=@n1, bAlphaBoolEnabled=True, Translation="0,0.5,0", Rotation="0,0,0", Scale="1,1,1") {
    alphaInputType=Bool
    rotationMode=BMM_Additive
    rotationSpace=BCS_BoneSpace
    translationMode=BMM_Additive
    translationSpace=BCS_BoneSpace
    alphaBoolBlend {blendOption=Linear}
    boneToModify {boneName=fix_neck_02}
    translation {y=0.5}
   }
   n2 = Layered blend per bone(
    BasePose=Copy Pose From Mesh() {bCopyCurves=true, bUseAttachedParent=true}
    BlendPoses_0=Copy Pose From Mesh(SourceMeshComponent=Get Head_Mesh() -> (Head_Mesh:Skeletal Mesh Component Object Reference))
   ) {layerSetup [{branchFilters [
       {boneName=FACIAL_C_Chin}
       {boneName=FACIAL_R_ChinSide}
       {boneName=FACIAL_L_ChinSide}
       {boneName=FACIAL_C_Jawline}
       {boneName=FACIAL_C_Jaw}
       {boneName=FACIAL_L_UnderChin}
       {boneName=FACIAL_R_UnderChin}
       {boneName=FACIAL_C_UnderChin}
       {boneName=FACIAL_C_FacialRoot}
      ]}]}
   n3 = Control Rig(Source=@n2) {
    alphaBoolBlend {blendOption=Linear}
    controlRigAssetReference {blueprintRigClass=&/F_MED_NobleSpear_CtrlRig.F_MED_NobleSpear_CtrlRig_C}
   }
   n4 = "Transform (Modify) Bone - Bone: attach_backpack"!(ComponentPose=Local To Component(LocalPose=@n3), bAlphaBoolEnabled:bool=True, Translation="0,4.5,0", Rotation="0,0,0", Scale="1,1,1") -> (Pose) {
    alphaInputType=Bool
    rotationMode=BMM_Additive
    rotationSpace=BCS_BoneSpace
    translationMode=BMM_Additive
    translationSpace=BCS_BoneSpace
    alphaBoolBlend {blendOption=Linear}
    boneToModify {boneName=attach_backpack}
    translation {y=4.5}
   }
   Output Pose(Result=Component To Local(ComponentPose=@n4)) {layerGroup=None}
  ]}
 EventGraph {nodes [
   Event OnCharacterCustomizationLoaded()
   > n1 = Cast To PlayerPawn_Athena(Object=TryGetPawnOwner())
   > Set Head_Mesh(Head_Mesh:Skeletal Mesh Component Object Reference=GetSkeletalMeshForPartType(self=@n1, PartType=Head)) -> (Output_Get:Skeletal Mesh Component Object Reference)
   Event BlueprintUpdateAnimation()
  ]}
]
variables [Head_Mesh {type=SkeletalMeshComponent}]
