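# Physics asset for the F_MED_NobleSpear face accessory mesh, providing physics bodies for five dynamic hair chains on the player cosmetic.
# Center, left, right and two side chains each run from a tiny base capsule on the head through two dyn_hair capsules with tight swing limits and zero-degree twist.
# Capsules on head, neck, upper spine, clavicles and upper arms form the body volume, linked by free-rotation constraints; the right upper arm carries two capsules.
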PhysicsAsset /BRCosmetics/Characters/Player/Female/Medium/Bodies/F_MED_NobleSpear/Meshes/Parts/F_MED_NobleSpear_FaceAcc_Physics
bodies {
 base_hair_cr [{shapeType=Capsule, length=0.505, radius=0.505, shapeName=base_hair_cr_capsule, rotation=<rot 0 -90 0>}]
 base_hair_lf [{shapeType=Capsule, length=0.505, radius=0.505, shapeName=base_hair_lf_capsule, rotation=<rot 0 -90 0>}]
 base_hair_rt [{shapeType=Capsule, length=0.505, radius=0.505, shapeName=base_hair_rt_capsule, rotation=<rot 0 -90 0>}]
 base_hair_side_lf [{shapeType=Capsule, length=0.505, radius=0.505, shapeName=base_hair_side_lf_capsule, rotation=<rot 0 -90 0>}]
 base_hair_side_rt [{shapeType=Capsule, length=0.505, radius=0.505, shapeName=base_hair_side_rt_capsule, rotation=<rot 0 -90 0>}]
 clavicle_l [{shapeType=Capsule, length=6.172, radius=7.8328, shapeName=clavicle_l_capsule, center=<5.2048 0 -3.2806>, rotation=<rot 80 -0.0004 -0.0004>}]
 clavicle_r [{shapeType=Capsule, length=6.172, radius=7.8328, shapeName=clavicle_r_capsule, center=<-5.1668 0 3.4961>, rotation=<rot 80 -0.0004 179.9996>}]
 dyn_hair_01_cr [{shapeType=Capsule, length=4, radius=3, shapeName=dyn_hair_01_cr_capsule, center=<5 0 0>, rotation=<rot 90 0 0>}]
 dyn_hair_01_lf [{shapeType=Capsule, length=4, radius=3, shapeName=dyn_hair_01_lf_capsule, center=<5 0 0>, rotation=<rot 90 0 0>}]
 dyn_hair_01_rt [{shapeType=Capsule, length=4, radius=3, shapeName=dyn_hair_01_rt_capsule, center=<5 0 0>, rotation=<rot 90 0 0>}]
 dyn_hair_02_cr [{shapeType=Capsule, length=6.4881, radius=3, shapeName=dyn_hair_02_cr_capsule, center=<5.1236 0 0.0218>, rotation=<rot 70.0001 180 180>}]
 dyn_hair_02_lf [{shapeType=Capsule, length=9.5999, radius=3, shapeName=dyn_hair_02_lf_capsule, center=<2.9942 0 -0.1457>, rotation=<rot 80 180 180>}]
 dyn_hair_02_rt [{shapeType=Capsule, length=4.0121, radius=3, shapeName=dyn_hair_02_rt_capsule, center=<4.0933 0 -1.3176>, rotation=<rot 80.0001 180 179.9999>}]
 dyn_hair_side_01_lf [{shapeType=Capsule, length=4, radius=3, shapeName=dyn_hair_side_01_lf_capsule, center=<5 0 0>, rotation=<rot 90 0 0>}]
 dyn_hair_side_01_rt [{shapeType=Capsule, length=4, radius=3, shapeName=dyn_hair_side_01_rt_capsule, center=<5 0 0>, rotation=<rot 90 0 0>}]
 dyn_hair_side_02_lf [{
   shapeType=Capsule
   length=4.131
   radius=2.0608
   shapeName=dyn_hair_side_02_lf_capsule
   center=<4.1733 -1.4864 0>
   rotation=<rot 90 -175.2364 -135.2364>
  }]
 dyn_hair_side_02_rt [{
   shapeType=Capsule
   length=2.1658
   radius=2.3726
   shapeName=dyn_hair_side_02_rt_capsule
   center=<3.9861 1.2054 0>
   rotation=<rot 0.0005 -140 -179.9998>
  }]
 head [{shapeType=Capsule, length=3.6382, radius=6.5146, center=<3.1401 4.6088 0.7554>, rotation=<rot 2.8054 89.7464 -85.9869>}]
 neck_01 [{shapeType=Capsule, length=8.7126, radius=4.7648, shapeName=neck_01_capsule, center=<4.2913 0.9616 0>, rotation=<rot 90 0 0>}]
 spine_05 [{shapeType=Capsule, length=13.2548, radius=7.3906, shapeName=spine_05_capsule, center=<1.9995 -1.4542 0>, rotation=<rot 0 -90 0>}]
 upperarm_l [{shapeType=Capsule, length=14.0738, radius=4.0219, shapeName=upperarm_l_capsule, center=<9.4153 0 0>, rotation=<rot 90 0 0>}]
 upperarm_r [
  {shapeType=Capsule, length=12.836, radius=5.2598, shapeName=upperarm_r_capsule, center=<-9.4153 0 0>, rotation=<rot 90 0 180>}
  {shapeType=Capsule, length=12.836, radius=5.2598, shapeName=upperarm_r_capsule, center=<-9.4153 -1.4683 -6.3324>, rotation=<rot 90 0 180>}
 ]
}
constraints [
 {
  bone1Name=upperarm_l
  bone2Name=clavicle_l
  swing1LimitDegrees=45
  swing1Motion=Free
  swing2LimitDegrees=45
  swing2Motion=Free
  twistLimitDegrees=45
  twistMotion=Free
 }
 {
  bone1Name=upperarm_r
  bone2Name=clavicle_r
  swing1LimitDegrees=45
  swing1Motion=Free
  swing2LimitDegrees=45
  swing2Motion=Free
  twistLimitDegrees=45
  twistMotion=Free
 }
 {
  bone1Name=head
  bone2Name=neck_01
  swing1LimitDegrees=45
  swing1Motion=Free
  swing2LimitDegrees=45
  swing2Motion=Free
  twistLimitDegrees=45
  twistMotion=Free
 }
 {
  bone1Name=clavicle_l
  bone2Name=spine_05
  swing1LimitDegrees=45
  swing1Motion=Free
  swing2LimitDegrees=45
  swing2Motion=Free
  twistLimitDegrees=45
  twistMotion=Free
 }
 {
  bone1Name=clavicle_r
  bone2Name=spine_05
  swing1LimitDegrees=45
  swing1Motion=Free
  swing2LimitDegrees=45
  swing2Motion=Free
  twistLimitDegrees=45
  twistMotion=Free
 }
 {
  bone1Name=neck_01
  bone2Name=spine_05
  swing1LimitDegrees=45
  swing1Motion=Free
  swing2LimitDegrees=45
  swing2Motion=Free
  twistLimitDegrees=45
  twistMotion=Free
 }
 {
  bone1Name=base_hair_cr
  bone2Name=head
  swing1LimitDegrees=45
  swing1Motion=Free
  swing2LimitDegrees=45
  swing2Motion=Free
  twistLimitDegrees=45
  twistMotion=Free
 }
 {bone1Name=dyn_hair_01_cr, bone2Name=base_hair_cr, swing1LimitDegrees=0, swing2LimitDegrees=5, twistLimitDegrees=0}
 {bone1Name=dyn_hair_02_cr, bone2Name=dyn_hair_01_cr, swing1LimitDegrees=0, swing2LimitDegrees=15, twistLimitDegrees=0}
 {
  bone1Name=base_hair_lf
  bone2Name=head
  swing1LimitDegrees=45
  swing1Motion=Free
  swing2LimitDegrees=45
  swing2Motion=Free
  twistLimitDegrees=45
  twistMotion=Free
 }
 {bone1Name=dyn_hair_01_lf, bone2Name=base_hair_lf, swing1LimitDegrees=0, swing2LimitDegrees=5, twistLimitDegrees=0}
 {bone1Name=dyn_hair_02_lf, bone2Name=dyn_hair_01_lf, swing1LimitDegrees=0, swing2LimitDegrees=15, twistLimitDegrees=0}
 {
  bone1Name=base_hair_rt
  bone2Name=head
  swing1LimitDegrees=45
  swing1Motion=Free
  swing2LimitDegrees=45
  swing2Motion=Free
  twistLimitDegrees=45
  twistMotion=Free
 }
 {bone1Name=dyn_hair_01_rt, bone2Name=base_hair_rt, swing1LimitDegrees=0, swing2LimitDegrees=5, twistLimitDegrees=0}
 {bone1Name=dyn_hair_02_rt, bone2Name=dyn_hair_01_rt, swing1LimitDegrees=0, swing2LimitDegrees=15, twistLimitDegrees=0}
 {
  bone1Name=base_hair_side_lf
  bone2Name=head
  swing1LimitDegrees=45
  swing1Motion=Free
  swing2LimitDegrees=45
  swing2Motion=Free
  twistLimitDegrees=45
  twistMotion=Free
 }
 {bone1Name=dyn_hair_side_01_lf, bone2Name=base_hair_side_lf, swing1LimitDegrees=5, swing2LimitDegrees=15, twistLimitDegrees=0}
 {bone1Name=dyn_hair_side_02_lf, bone2Name=dyn_hair_side_01_lf, swing1LimitDegrees=10, swing2LimitDegrees=10, twistLimitDegrees=0}
 {
  bone1Name=base_hair_side_rt
  bone2Name=head
  swing1LimitDegrees=45
  swing1Motion=Free
  swing2LimitDegrees=45
  swing2Motion=Free
  twistLimitDegrees=45
  twistMotion=Free
 }
 {bone1Name=dyn_hair_side_01_rt, bone2Name=base_hair_side_rt, swing1LimitDegrees=5, swing2LimitDegrees=5, twistLimitDegrees=0}
 {bone1Name=dyn_hair_side_02_rt, bone2Name=dyn_hair_side_01_rt, swing1LimitDegrees=3, swing2LimitDegrees=10, twistLimitDegrees=0}
]
